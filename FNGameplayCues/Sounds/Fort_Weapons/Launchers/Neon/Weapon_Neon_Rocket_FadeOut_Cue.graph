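# Sound cue for the fade-out sound of the Neon rocket launcher.
# A single wave player feeds the Neon_Rocket_Charge_Fade wave straight to the output, with no randomization or modulation.
# It borrows bullet-impact settings: the Fort_Impacts sound class, bullet impact attenuation and rifle bullet impact concurrency.
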SoundCue /FNGameplayCues/Sounds/Fort_Weapons/Launchers/Neon/Weapon_Neon_Rocket_FadeOut_Cue
root=WavePlayer_0
nodes [{id=WavePlayer_0, properties {soundWaveAssetPtr=&/Neon_Rocket_Charge_Fade}}]
properties {
 attenuationSettings=/Game/Sounds/Attenuation/Fort_Bullet_Impact_Att
 maxDistance=5300
 soundClassObject=&Fort_Impacts
 volumeMultiplier=1
 concurrencySet [&RifleBulletImpact_Conc]
}
